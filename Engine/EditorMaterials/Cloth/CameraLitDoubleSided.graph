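# Unlit two-sided editor material for cloth and skeletal meshes that fakes camera-based lighting with distinct front and back tints.
# The dot product of the normalized vertex-to-camera vector and vertex normal supplies brightness through its absolute value.
# Its sign drives the lerp between Back and Front colors, so negative values extrapolate past Back Color.
# The shaded tint is multiplied by vertex color.
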Material /Engine/EditorMaterials/Cloth/CameraLitDoubleSided
graph {
 properties {
  bDisableDepthTest=true
  bUsedWithClothing=true
  bUsedWithEditorCompositing=true
  bUsedWithMeshDeformer=true
  bUsedWithSkeletalMesh=true
  pixelDepthOffsetMode=PDOM_Legacy
  shadingModel=MSM_Unlit
  translucencyPass=MTP_BeforeDOF
  twoSided=true
  usedShadingModels=Unlit
 }
 nodes [
  n1 = Constant3Vector() {desc=Back Color, constant=<rgba 0.0497 0.305 0.3278 0>}
  n2 = Constant3Vector() {desc=Front Color, constant=<rgba 0.4852 0.1812 0.0252 0>}
  n3 = CameraPositionWS()
  n4 = WorldPosition()
  n5 = DotProduct(
   A=MaterialFunctionCall(Vector=Subtract(A=@n3, B=@n4.XYZ), Default=Constant3Vector() {constant=<rgba -1 0 0 0>}) {desc=VertexToCamera, materialFunction=&SafeNormalize}.Result
   B=VertexNormalWS()
  )
  n6 = Sign(None=@n5)
  Multiply(A=Multiply(A=LinearInterpolate(A=@n1, B=@n2, Alpha=@n6), B=Abs(None=@n5)), B=VertexColor())
 ]
}
