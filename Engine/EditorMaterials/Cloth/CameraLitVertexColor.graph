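# Unlit, two-sided editor material for clothing and skeletal meshes that displays vertex colors with simple camera-based shading.
# It dots the safe-normalized surface-to-camera vector with the vertex normal and takes the absolute value, so faces pointing toward or away from the camera both appear bright.
# That facing term scales a white base color and the vertex color, faking a headlight without scene lighting.
# Depth testing is flagged as disabled.
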Material /Engine/EditorMaterials/Cloth/CameraLitVertexColor
graph {
 properties {
  bDisableDepthTest=true
  bUsedWithClothing=true
  bUsedWithEditorCompositing=true
  bUsedWithMeshDeformer=true
  bUsedWithSkeletalMesh=true
  pixelDepthOffsetMode=PDOM_Legacy
  shadingModel=MSM_Unlit
  translucencyPass=MTP_BeforeDOF
  twoSided=true
  usedShadingModels=Unlit
 }
 nodes [
  n1 = CameraPositionWS()
  n2 = WorldPosition()
  n3 = DotProduct(
   A=MaterialFunctionCall(Vector=Subtract(A=@n1, B=@n2.XYZ), Default=Constant3Vector() {constant=<rgba -1 0 0 0>}) {desc=VertexToCamera, materialFunction=&SafeNormalize}.Result
   B=VertexNormalWS()
  )
  Multiply(A=Multiply(A=Constant3Vector() {desc=BaseColor, constant=<rgba 1 1 1 0>}, B=Abs(None=@n3)), B=VertexColor())
 ]
}
